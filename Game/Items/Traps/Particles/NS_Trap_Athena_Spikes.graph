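# Niagara system for the spike trap, a short burst of translucent pale gray sprites, likely a dust puff.
# The FN_Trap_Spawn emitter fires an instant burst, with custom trap and sprite modules setting speed, lifetime, size and rotation.
# Drag slows the particles as curves grow them, fade alpha and lower a dynamic material parameter, alongside SubUV animation driven by normalized age.
# A second, smaller and slower emitter, FN_Trap_Spawn001, is disabled.
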NiagaraSystem /Game/Items/Traps/Particles/NS_Trap_Athena_Spikes
dependencies {
 usedDataInterfaces [&NiagaraDataInterfaceCurve]
 usedDynamicInputs []
 usedModules [/Niagara/Modules/System/SystemLifeCycle, &SystemState]
}
emitters {
 FN_Trap_Spawn {
  emitterUpdateScript [
   &EmitterScalability(
    Module.bApplySpawnCountScale=false
    Module.bApplySpawnCountScaleByDistanceFraction=false
    Module.bUseMaxDistance=false
    Module.bUseMinDistance=false
   ) {enabled=false}
   &EmitterLifeCycle() {enabled=false}
   &EmitterState()
   &SpawnBurst_Instantaneous(Spawn Count=100, Use Spawn Probability=false)
  ]
  particleSpawnScript [
   /Game/Effects/Niagara/Modules/Trap/FN_Initialize_Trap(
    Color:LinearColor=<rgba 0.8 0.8 0.8 0.1>
    Edge Length=1
    Fan=1
    Lifetime:NiagaraFloat=&UniformRangedFloat(Minimum=0.2, Maximum=0.75)
    Speed Range:Vector2f=<500 1000>
   )
   /Game/Effects/Niagara/Modules/Sprite/FN_InitializeSpriteAttributes(
    Module.Write SpriteRotation=true
    Module.Write SpriteSize=true
    Sprite Rotation:NiagaraFloat=&UniformRangedFloat(Minimum=0, Maximum=360)
    Sprite Size:Vector2f=&Vector2DFromFloat(Value=&UniformRangedFloat(Minimum=100, Maximum=250))
   )
  ]
  particleUpdateScript [
   &ParticleState()
   &ScaleSpriteSize(Scale Factor=&Vector2DFromFloat(Value=&FloatFromCurve(FloatCurve=DataInterface() {propertyValues {Curve {keys [{arriveTangent=1.4897, interpMode=RCIM_Cubic, leaveTangent=1.4897, tangentMode=RCTM_User, time=0.0014, value=0.4607}, <at 1.0006 0.9823>]}}})))
   &Drag(Drag=10, Use Linear Drag=true, Use Rotational Drag=false)
   &SolveForcesAndVelocity(Clamp Velocity=false, Limit Acceleration=false)
   /Niagara/Modules/Update/SubUV/SubUVAnimation(Number Of Frames=64, SubUV Lookup Index=&Multiply_Float(A=$Particles.NormalizedAge, B=0.1))
   &Color(
    Color=$Particles.Initial.Color
    Scale Alpha=&FloatFromCurve(FloatCurve=DataInterface() {propertyValues {Curve {keys [<at -0.0046 1.0113>,<at 1.0011 0.002>]}}})
   )
   &DynamicMaterialParameters(
    Index 0 Param 1=&FloatFromCurve(FloatCurve=DataInterface() {propertyValues {Curve {keys [<at -0.0342 0.3114>,<at 0.9804 -0.6636>]}}})
    Index 0 Param 2=1
    Index 0 Param 3=1
    Index 0 Param 4=1
    Module.Param0WriteEnabled=true
    Module.Param1WriteEnabled=true
    Module.Param2WriteEnabled=true
    Module.Param3WriteEnabled=true
   )
  ]
  rendererClasses [&NiagaraSpriteRendererProperties]
  renderers [{rendererClass=&NiagaraSpriteRendererProperties}]
 }
 FN_Trap_Spawn001 {
  bEnabled=false
  emitterUpdateScript [
   &EmitterScalability(
    Module.bApplySpawnCountScale=false
    Module.bApplySpawnCountScaleByDistanceFraction=false
    Module.bUseMaxDistance=false
    Module.bUseMinDistance=false
   )
   &EmitterLifeCycle()
  ]
  particleSpawnScript [
   &FN_Initialize_Trap(Edge Length=1, Fan=0.4351, Lifetime=0.5, Pitch Range:Vector2f=<0 2>, Speed Range:Vector2f=<100 300>)
   &FN_InitializeSpriteAttributes(
    Module.Write SpriteRotation=true
    Module.Write SpriteSize=true
    Sprite Rotation:NiagaraFloat=&UniformRangedFloat(Minimum=0, Maximum=360)
    Sprite Size:Vector2f=&Vector2DFromFloat(Value=&UniformRangedFloat(Minimum=50, Maximum=100))
   )
  ]
  particleUpdateScript [
   &ParticleState()
   &ScaleSpriteSize(Scale Factor=&Vector2DFromFloat(Value=&FloatFromCurve(FloatCurve=DataInterface() {propertyValues {Curve {keys [{arriveTangent=1.4897, interpMode=RCIM_Cubic, leaveTangent=1.4897, tangentMode=RCTM_User, time=0.0014, value=0.4607}, <at 1.0006 0.9823>]}}})))
   &SolveForcesAndVelocity(Clamp Velocity=false, Limit Acceleration=false)
   &SubUVAnimation(Number Of Frames=64, SubUV Lookup Index=&Multiply_Float(A=$Particles.Age, B=0.5833))
   &Color(
    Color=$Particles.Initial.Color
    Scale Alpha=&FloatFromCurve(FloatCurve=DataInterface() {propertyValues {Curve {keys [<at -0.0046 1.0113>,<at 1.0011 0.002>]}}})
   )
   &DynamicMaterialParameters(
    Index 0 Param 1=0.6
    Index 0 Param 2=1
    Index 0 Param 3=1
    Index 0 Param 4=0.6455
    Module.Param0WriteEnabled=true
    Module.Param1WriteEnabled=true
    Module.Param2WriteEnabled=true
    Module.Param3WriteEnabled=true
   )
  ]
  rendererClasses [&NiagaraSpriteRendererProperties]
  renderers [{rendererClass=&NiagaraSpriteRendererProperties}]
 }
}
summary {
 systemName=NS_Trap_Athena_Spikes
 emitters [
  {emitterName=FN_Trap_Spawn, rendererClasses [&NiagaraSpriteRendererProperties]}
  {bEnabled=false, emitterName=FN_Trap_Spawn001, rendererClasses [&NiagaraSpriteRendererProperties]}
 ]
 userVariables []
}
